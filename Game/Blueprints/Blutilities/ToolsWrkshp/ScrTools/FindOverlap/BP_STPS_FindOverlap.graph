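# Editor utility Blueprint that serves as the options and command object for a Find Overlap level-editing tool.
# Exposes settings for ignoring devices, overlap precision, showing only overlapping objects and resetting the view on exit.
# Four functions (find all overlapping objects, focus next, select all, reset view) each only broadcast a matching event dispatcher.
# The file holds no detection or visibility logic; the event graph is empty.
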EditorUtilityBlueprint /Game/Blueprints/Blutilities/ToolsWrkshp/ScrTools/FindOverlap/BP_STPS_FindOverlap
graphs [
 EventFindOverlappingObjectsClicked {nodes [EventFindOverlappingObjectsClicked() {is=title K2Node_FunctionEntry}]}
 EventFocusNextOverlapClicked {nodes [EventFocusNextOverlapClicked() {is=title K2Node_FunctionEntry}]}
 EventGraph {links [], nodes []}
 EventOnSelectAllOverlappingClicked {nodes [EventOnSelectAllOverlappingClicked() {is=title K2Node_FunctionEntry}]}
 EventRestoreVisibilityClicked {nodes [EventRestoreVisibilityClicked() {is=title K2Node_FunctionEntry}]}
 Find All Overlapping Objects {nodes [
   Find All Overlapping Objects() {is=title K2Node_FunctionEntry}
   > Call Event Find Overlapping Objects Clicked!(execute:Exec, self:BP STPS Find Overlap Object Reference) -> (then:Exec) {is=title K2Node_CallDelegate}
  ]}
 Focus Next Overlap {nodes [
   Focus Next Overlap() {is=title K2Node_FunctionEntry}
   > Call Event Focus Next Overlap Clicked!(execute:Exec, self:BP STPS Find Overlap Object Reference) -> (then:Exec) {is=title K2Node_CallDelegate}
  ]}
 Reset View {nodes [
   Reset View() {is=title K2Node_FunctionEntry}
   > Call Event Restore Visibility Clicked!(execute:Exec, self:BP STPS Find Overlap Object Reference) -> (then:Exec) {is=title K2Node_CallDelegate}
  ]}
 Select All Overlapping {nodes [
   Select All Overlapping() {is=title K2Node_FunctionEntry}
   > Call Event on Select All Overlapping Clicked!(execute:Exec, self:BP STPS Find Overlap Object Reference) -> (then:Exec) {is=title K2Node_CallDelegate}
  ]}
]
variables [
 EventFindOverlappingObjectsClicked {default=()}
 EventFocusNextOverlapClicked {default=()}
 EventOnSelectAllOverlappingClicked {default=()}
 EventRestoreVisibilityClicked {default=()}
 Ignore Devices {type=boolean, default=true}
 Overlap Threshold {type=E_Precision_FindOverlap, default=Precise}
 Reset View On Exit {type=boolean, default=true}
 Show Only Overlapping {type=boolean}
]
