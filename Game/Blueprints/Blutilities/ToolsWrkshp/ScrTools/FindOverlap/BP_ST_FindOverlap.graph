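# Editor scriptable tool that finds overlapping or near-duplicate actors in the level, so they can be reviewed and selected.
# The scan skips ignored names, classes and optionally devices, then groups each actor with its nearest neighbor using precise or rough distance, bounds and transform thresholds.
# Line sets mark groups with boxes and vertical beams, while header widgets show counts and actions.
# Hotkeys cycle focus, toggle unlit isolation, and restore saved visibility and camera.
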EditorUtilityBlueprint /Game/Blueprints/Blutilities/ToolsWrkshp/ScrTools/FindOverlap/BP_ST_FindOverlap
graphs [
 EventGraph {nodes [
   Event OnScriptSetup()
   > n1 = Sequence!(execute) -> (
    then_0
    then_1
    then_2:Exec
    then_3:Exec
    then_4:Exec
    then_5:Exec
    then_6:Exec
    then_7:Exec
    then_8:Exec
    then_9:Exec
    then_10:Exec
    then_11:Exec
    then_12:Exec
   )
   Set Total Clusters Overlapping(execute=@n1.then_1, Total Clusters Overlapping:int=0) -> (Output_Get:int=0)
   > Set CurClusterFocused(CurClusterFocused:int=0) -> (Output_Get:int=0)
   > Set Total Actors Overlapping(Total Actors Overlapping:int=0) -> (Output_Get:int=0)
   > Clear!(
    execute
    TargetMap:Map of Actor Object References to S Overlap Data Find Overlap Structures=Get All Actors Filtered Descriptions() -> (All Actors Filtered Descriptions:Map of Actor Object References to S Overlap Data Find Overlap Structures)
   ) -> (then) {is=title K2Node_CallFunction}
   > Clear(TargetMap:Map of Strings to S Overlap Data Find Overlap Structures=Get All Bounds Dict() -> (All Bounds Dict:Map of Strings to S Overlap Data Find Overlap Structures))
   > Clear(TargetMap=Get Overlapping Bounds Dict() -> (Overlapping Bounds Dict:Map of Strings to S Overlap Data Find Overlap Structures))
   > Clear(TargetArray:Array of Actor Object References=Get AllOverlappingActors() -> (AllOverlappingActors:Array of Actor Object References)) {is=title K2Node_CallArrayFunction}
   > n2 = GetEditorActiveViewportViewMode!(execute) -> (then, ReturnValue:EViewModeIndex Enum=VMI_BrushWireframe)
   > Set View Mode Before OnShowOnly(View Mode Before OnShowOnly:EViewModeIndex Enum=@n2) -> (Output_Get:EViewModeIndex Enum=VMI_BrushWireframe)
   n3 = Construct Object from Class(
    execute=@n1.then_11
    Class=&BP_UW_SharedHeaderWidget_C
   )
   > n4 = Set Header Widget(Header Widget:BP UW Shared Header Widget Object Reference=@n3) -> (Output_Get:BP UW Shared Header Widget Object Reference)
   > SetToolPanelToolHeaderWidget(Widget=@n4)
   n5 = AddPropertySetOfType(
    execute=@n1.then_2
    PropertySetType=/Game/Blueprints/Blutilities/ToolsWrkshp/ScrTools/Shared/SharedPropertySets/BP_STPS_SharedWidgetState.BP_STPS_SharedWidgetState_C
    Identifier=Get SharedStateIdentifier() -> (SharedStateIdentifier:String)
   )
   > n6 = Cast To BP_STPS_SharedWidgetState!(execute, Object=@n5) -> (then, CastFailed, AsBP STPS Shared Widget State:BP STPS Shared Widget State Object Reference) {is=title K2Node_DynamicCast}
   > Set SharedState Settings(SharedState Settings:BP STPS Shared Widget State Object Reference=@n6) -> (Output_Get:BP STPS Shared Widget State Object Reference)
   > SetPropertySetVisibleByName!(execute, self:Scriptable Interactive Tool Object Reference, Identifier:String=Get SharedStateIdentifier(), bVisible:bool=false) -> (then)
   > RestorePropertySetSettings(
    PropertySet=Get SharedState Settings() -> (SharedState Settings:BP STPS Shared Widget State Object Reference)
    SaveKey=Get Property Set Save Key() -> (Property Set Save Key:String)
   )
   n7 = AddPropertySetOfType(
    execute=@n1.then_3
    PropertySetType=&/BP_STPS_FindOverlap.BP_STPS_FindOverlap_C
    Identifier=Get Settings Identifier() -> (Settings Identifier:String)
   )
   > n8 = Cast To BP_STPS_FindOverlap!(execute, Object=@n7) -> (then, CastFailed, AsBP STPS Find Overlap:BP STPS Find Overlap Object Reference) {is=title K2Node_DynamicCast}
   > n9 = Set Settings(Settings:BP STPS Find Overlap Object Reference=@n8) -> (Output_Get:BP STPS Find Overlap Object Reference)
   > RestorePropertySetSettings(PropertySet=@n9, SaveKey=Get Property Set Save Key())
   AddMouseHoverBehavior!(
    execute=@n1.then_4
    self:Scriptable Modular Behavior Tool Object Reference
    BeginHoverSequenceHitTest:Delegate=Create Event()
    OnBeginHover:Delegate
    OnUpdateHover:Delegate
    OnEndHover:Delegate
    HoverCaptureCheck:Delegate
    CapturePriority:int=100
   ) -> (then)
   > AddSingleKeyInputBehavior!(
    execute
    self:Scriptable Modular Behavior Tool Object Reference
    OnKeyPressed:Delegate=Create Event()
    OnKeyReleased:Delegate
    OnForceEndCaptureFuncIn:Delegate
    Key:Key Structure=Get Focus Next Hotkey() -> (Focus Next Hotkey:Key Structure)
    CaptureCheck:Delegate
    CapturePriority:int=0
   ) -> (then)
   > AddSingleKeyInputBehavior(OnKeyPressed=Create Event(), Key=Get Cycle Overlapping Viz Key() -> (Cycle Overlapping Viz Key:Key Structure), CapturePriority=0)
   > AddSingleKeyInputBehavior(OnKeyPressed=Create Event(), Key=Get Reset View Hotkey() -> (Reset View Hotkey:Key Structure), CapturePriority=0)
   Set Show Only Overlapping!(
    execute=@n1.then_5
    Show Only Overlapping:bool=false
    self:BP STPS Find Overlap Object Reference=Get Settings() -> (Settings:BP STPS Find Overlap Object Reference)
   ) -> (then, Output_Get:bool)
   n10 = AddLineSet!(execute=@n1.then_6, self:Scriptable Interactive Tool Object Reference) -> (then, ReturnValue:Scriptable Tool Line Set Object Reference)
   > Set Line Set For Indicators(Line Set For Indicators:Scriptable Tool Line Set Object Reference=@n10) -> (Output_Get:Scriptable Tool Line Set Object Reference)
   > n11 = AddLineSet()
   > Set Line Set For Bbox(Line Set For Bbox:Scriptable Tool Line Set Object Reference=@n11) -> (Output_Get:Scriptable Tool Line Set Object Reference)
   n12 = GetActorsTempHidden!(execute=@n1.then_7) -> (then, ActorTempHiddenMap:Map of Actor Object References to Booleans, OnlyTempHidden:Array of Actor Object References)
   > Set Actor Temp Hidden At Start(Actor Temp Hidden At Start:Map of Actor Object References to Booleans=@n12) -> (Output_Get:Map of Actor Object References to Booleans)
   Bind Event to Event Find Overlapping Objects Clicked(execute=@n1.then_8, self:BP STPS Find Overlap Object Reference=Get Settings(), Delegate=Create Event()) {is=title K2Node_AddDelegate}
   > Bind Event to Event Restore Visibility Clicked(self:BP STPS Find Overlap Object Reference=Get Settings(), Delegate=Create Event()) {is=title K2Node_AddDelegate}
   > Bind Event to Event Focus Next Overlap Clicked(self:BP STPS Find Overlap Object Reference=Get Settings(), Delegate=Create Event()) {is=title K2Node_AddDelegate}
   > Bind Event to Event on Select All Overlapping Clicked(self:BP STPS Find Overlap Object Reference=Get Settings(), Delegate=Create Event()) {is=title K2Node_AddDelegate}
   WatchProperty(execute=@n1.then_9, PropertySet=Get Settings(), PropertyName=Show Only Overlapping, OnModified=Create Event())
   n13 = GetLevelViewportCameraInfo!(self:Unreal Editor Subsystem Object Reference=Get UnrealEditorSubsystem()) -> (CameraLocation:Vector, CameraRotation:Rotator, ReturnValue:bool)
   Set Viewport Cam At Start Location(execute=@n1.then_10, Viewport Cam At Start Location:Vector=@n13) -> (Output_Get:Vector)
   > Set Viewport Cam At Start Rotation(Viewport Cam At Start Rotation:Rotator=@n13.CameraRotation) -> (Output_Get:Rotator)
   Event OnScriptShutdown()
   > n14 = Sequence!(execute) -> (then_0, then_1, then_2, then_3)
   > Branch(Condition=Get Reset View On Exit!(self:BP STPS Find Overlap Object Reference=Get Settings()) -> (Reset View On Exit:bool))
   > OnRestoreVisibilityClicked()
   Set Show Only Overlapping(execute=@n14.then_1, Show Only Overlapping=false, self=Get Settings())
   SavePropertySetSettings(execute=@n14.then_2, PropertySet=Get Settings(), SaveKey=Get Property Set Save Key())
   n15 = Get Readout Box() -> (Readout Box:BP UW Expandable Box Object Reference)
   > Set ReadoutOpenOnExit!(
    execute
    ReadoutOpenOnExit:bool=IsChecked(self=Get ExpandButton!(self:BP UW Expandable Box Object Reference=@n15) -> (ExpandButton:Check Box Object Reference))
    self:BP STPS Shared Widget State Object Reference=Get SharedState Settings()
   ) -> (then, Output_Get:bool)
   > SavePropertySetSettings(PropertySet=Get SharedState Settings(), SaveKey=Get Property Set Save Key())
   RemoveAllLines(
    execute=@n14.then_3
    self:Scriptable Tool Line Set Object Reference=Get Line Set For Indicators() -> (Line Set For Indicators:Scriptable Tool Line Set Object Reference)
   )
   > RemoveAllLines(self=Get Line Set For Bbox() -> (Line Set For Bbox:Scriptable Tool Line Set Object Reference))
   > Branch(Condition=Get Reset View On Exit(self=Get Settings()))
   > SetLevelViewportCameraInfo!(
    execute
    self:Unreal Editor Subsystem Object Reference=Get UnrealEditorSubsystem()
    CameraLocation:Vector=Get Viewport Cam At Start Location() -> (Viewport Cam At Start Location:Vector)
    CameraRotation:Rotator=Get Viewport Cam At Start Rotation() -> (Viewport Cam At Start Rotation:Rotator)
   ) -> (then)
   n16 = Event OnScriptTick()
   > Set CurTime(CurTime=float + float(A=@n16.DeltaTime, B=Get CurTime()))
   > n17 = Branch(Condition=Get FirstInitRun())
   Set FirstInitRun(execute=@n17.else)
   > n18 = Sequence!(execute) -> (then_0, then_1, then_2)
   n19 = Get Header Widget()
   > n20 = AddExpandableBox(self=@n19)
   > n21 = Set Readout Box(Readout Box:BP UW Expandable Box Object Reference=@n20) -> (Output_Get:BP UW Expandable Box Object Reference)
   > SetUpBox(
    self=@n21
    Name=Get Readout Box Label!(self:BP UW Shared Header Widget Object Reference=@n19) -> (Readout Box Label:Text)
    OpenByDefault=Get ReadoutOpenOnExit!(self:BP STPS Shared Widget State Object Reference=Get SharedState Settings()) -> (ReadoutOpenOnExit:bool)
   )
   > n22 = Sequence!(execute) -> (then_0, then_1)
   n23 = Set Overlap Groups Line(execute=@n24.then, Overlap Groups Line:BP UW Readout Line Object Reference=@n24) -> (Output_Get:BP UW Readout Line Object Reference)
   n24 = SetupLine!(
    execute=@n22.then_0
    self:BP UW Expandable Box Object Reference=Get Readout Box()
    CurrentLine:BP UW Readout Line Object Reference=@n23
    DynamicLine:Vertical Box Object Reference
    Label:Text="Get Widget Num Overlap Groups  Label"() -> ("Widget Num Overlap Groups  Label":Text)
    Value:Text="Get Widget Num Overlap Groups  Value"() -> ("Widget Num Overlap Groups  Value":Text)
   ) -> (then, Readout Line:BP UW Readout Line Object Reference)
   n25 = Set Overlap Total Line(execute=@n26.then, Overlap Total Line:BP UW Readout Line Object Reference=@n26) -> (Output_Get:BP UW Readout Line Object Reference)
   n26 = SetupLine(
    execute=@n22.then_1
    self=Get Readout Box()
    CurrentLine=@n25
    Label="Get Widget Num Overlap Total  Label"() -> ("Widget Num Overlap Total  Label":Text)
    Value="Get Widget Num Overlap Total  Value"() -> ("Widget Num Overlap Total  Value":Text)
   )
   n27 = Get Header Widget()
   n28 = AddExpandableBox(execute=@n18.then_1, self=@n27)
   > n29 = Set Shortcut Box(Shortcut Box:BP UW Expandable Box Object Reference=@n28) -> (Output_Get:BP UW Expandable Box Object Reference)
   > n30 = SetUpBox(
    self=@n29
    Name=Get Shortcut Box Label!(self:BP UW Shared Header Widget Object Reference=@n27) -> (Shortcut Box Label:Text)
    OpenByDefault=Get ShortcutsOpenOnExit!(self:BP STPS Shared Widget State Object Reference=Get SharedState Settings()) -> (ShortcutsOpenOnExit:bool)
   )
   n31 = Set ShortcutLines(execute=@n32.then, ShortcutLines:Array of BP UW Readout Line Object References=@n32) -> (Output_Get:Array of BP UW Readout Line Object References)
   n32 = AddShortcuts!(
    execute=@n30.then
    self:BP UW Expandable Box Object Reference=Get Shortcut Box() -> (Shortcut Box:BP UW Expandable Box Object Reference)
    ShortcutMap:Map of Strings to Strings=Get Shortcut Map() -> (Shortcut Map:Map of Strings to Strings)
    ExistingShortcutLines:Array of BP UW Readout Line Object References=@n31
    ShortcutMapKeys:Array of Strings=Get Shortcut Map Keys() -> (Shortcut Map Keys:Array of Strings)
   ) -> (then, AllShortcutLines:Array of BP UW Readout Line Object References)
   n33 = Sequence(execute=@n18.then_2)
   > n34 = AddExpandableBox(self=Get Header Widget())
   > n35 = Set Actions Box(Actions Box=@n34)
   > SetUpBox(self=@n35, Name=Get Actions Box Label(self=Get Header Widget()))
   > n36 = Construct Object from Class(Class=/Game/Blueprints/Blutilities/ToolsWrkshp/ScrTools/FindOverlap/SupportFiles/BP_UW_ActionsButton_FindOverlap.BP_UW_ActionsButton_FindOverlap_C) -> (ReturnValue:BP UW Actions Button Find Overlap Object Reference)
   > n37 = Set ActionButtons(ActionButtons:BP UW Actions Button Find Overlap Object Reference=@n36) -> (Output_Get:BP UW Actions Button Find Overlap Object Reference)
   > AddChildToVerticalBox(self=Get DynamicEntries(self=Get Actions Box()), Content=@n37)
   Bind Event to Event Find All Overlapping Clicked(
    execute=@n33.then_1
    self:BP UW Actions Button Find Overlap Object Reference=Get ActionButtons() -> (ActionButtons:BP UW Actions Button Find Overlap Object Reference)
    Delegate=Create Event()
   ) {is=title K2Node_AddDelegate}
   > Bind Event to Event Select All Overlapping Clicked(self:BP UW Actions Button Find Overlap Object Reference=Get ActionButtons(), Delegate=Create Event()) {is=title K2Node_AddDelegate}
   > Bind Event to Event Reset View Clicked(self:BP UW Actions Button Find Overlap Object Reference=Get ActionButtons(), Delegate=Create Event()) {is=title K2Node_AddDelegate}
   > Bind Event to Event Focus Next Overlap Clicked(self:BP UW Actions Button Find Overlap Object Reference=Get ActionButtons(), Delegate=Create Event())
  ]}
 GetOverlappingActors {nodes [
   n1 = Get Overlapping Bounds Dict()
   n2 = PrintText!(
    execute
    InText=Format Text!(
     Format="LOCTEXT(\":\\r\\nMap len: {ml}\\r\\nNum Mult Actors: {nma}\")"
     ml:int=Length!(TargetMap:Map of Strings to S Overlap Data Find Overlap Structures=@n1) -> (ReturnValue:int=0) {is=title K2Node_CallFunction}
     nma:int=Get Num With Mult Actors() -> (Num With Mult Actors:int=0)
    ) -> (Result)
    bPrintToScreen
    bPrintToLog
    TextColor
    Duration
    Key
   ) -> (then)
   Return Node!(
    execute=[@n2.then, @n35.CompletedPin]
    All Overlapping Actors:Array of Actor Object References=Get AllOverlappingActors()
    Overlapping Bounds Dict:Map of Strings to S Overlap Data Find Overlap Structures=Get Overlapping Bounds Dict()
   )
   GetOverlappingActors() {is=title K2Node_FunctionEntry}
   > Clear(TargetMap=Get AllBoundsDict() -> (AllBoundsDict:Map of Strings to S Overlap Data Find Overlap Structures)) {is=title K2Node_CallFunction}
   > n3 = Sequence!(execute) -> (
    then_0
    then_1
    then_2:Exec
    then_3:Exec
    then_4:Exec
    then_5:Exec
    then_6:Exec
    then_7:Exec
    then_8:Exec
    then_9:Exec
    then_10:Exec
    then_11:Exec
   )
   Return Node(execute=@n3.then_11, All Overlapping Actors=Get AllOverlappingActors(), Overlapping Bounds Dict=Get Overlapping Bounds Dict())
   n4 = Switch on E_Precision_FindOverlap!(
    execute=@n3.then_2
    Selection:E_Precision_FindOverlap Enum=Get Overlap Threshold!(self:BP STPS Find Overlap Object Reference=Get Settings()) -> (Overlap Threshold:E_Precision_FindOverlap Enum=NewEnumerator0)
   ) -> (NewEnumerator0:Exec, NewEnumerator1:Exec) {is=title K2Node_SwitchEnum}
   > Set Dist Thresh(Dist Thresh:double=Get Precise Dist Thresh() -> (Precise Dist Thresh:double)) -> (Output_Get:double)
   > Set Bbox Thresh(Bbox Thresh:double=Get Precise Bbox Thresh() -> (Precise Bbox Thresh:double)) -> (Output_Get:double)
   > n5 = Set Xform Thresh(Xform Thresh:double=Get Precise Xform Thresh() -> (Precise Xform Thresh:double)) -> (Output_Get:double)
   Clear(execute=[@n34.then, @n5.then], TargetMap=Get Overlapping Bounds Dict())
   > Clear(TargetArray=Get AllOverlappingActors()) {is=title K2Node_CallArrayFunction}
   n6 = Get AllLevelActors() -> (AllLevelActors:Array of Actor Object References)
   > Set All Level Actors Not Seen(All Level Actors Not Seen:Array of Actor Object References=@n6) -> (Output_Get:Array of Actor Object References)
   > n7 = For Each Loop(Array:Array of Actor Object References=@n6) -> (Array Element:Actor Object Reference)
   > n8 = Set Current Actor(Current Actor:Actor Object Reference=@n7) -> (Output_Get:Actor Object Reference)
   n9 = Get IgnoreTheseActors() -> (IgnoreTheseActors:Array of Actor Object References)
   > n10 = Branch(Condition="integer >= integer"(A=Find Item!(TargetArray:Array of Actor Object References=@n9, ItemToFind:"Actor Object Reference (by ref)"=@n8) -> (ReturnValue)))
   > PrintText(InText=Format Text!(Format="LOCTEXT(\"Ignoring: {a}\")", a:Actor Object Reference=@n8) -> (Result), bPrintToScreen=false, bPrintToLog=false)
   n11 = GetObjectName(Object=Get Current Actor() -> (Current Actor:Actor Object Reference))
   Set Current Object Name(execute=@n10.else, Current Object Name:String=@n11) -> (Output_Get:String)
   > Remove Item!(
    execute
    TargetArray:Array of Actor Object References=Get All Level Actors Not Seen() -> (All Level Actors Not Seen:Array of Actor Object References)
    Item:"Actor Object Reference (by ref)"=Get Current Actor()
   ) -> (then, ReturnValue:bool)
   n12 = FindNearestActor!(Origin:Vector=Get Actor Location(self=@n8), ActorsToCheck:Array of Actor Object References=Get All Level Actors Not Seen()) -> (Distance:float, ReturnValue:Actor Object Reference)
   > n13 = Branch(Condition=Find(TargetMap:Map of Strings to S Overlap Data Find Overlap Structures=Get Overlapping Bounds Dict(), Key=GetObjectName(Object=@n12.ReturnValue)) -> (Value:S Overlap Data Find Overlap Structure).ReturnValue)
   n14 = Set Current Nearest(execute=@n13.else, Current Nearest:Actor Object Reference=@n12.ReturnValue) -> (Output_Get:Actor Object Reference)
   n15 = Find Item(TargetArray=Get IgnoreTheseActors(), ItemToFind=@n14)
   n16 = Get Dist Thresh() -> (Dist Thresh:double)
   > n17 = Branch(Condition=AND Boolean(A="float <= float"(A=@n12, B=@n16) -> (ReturnValue=~), B="integer < integer"(A=@n15)))
   n18 = Get Actor Bounds(self=Get Current Actor(), bOnlyCollidingComponents=false)
   Set Cur Actor Box(
    execute=[@n70.then, @n69.else, @n17.then]
    Cur Actor Box:Box Structure=MakeBoxWithOrigin!(Origin:"Vector (by ref)"=@n18, Extent:"Vector (by ref)"=@n18.BoxExtent) -> (ReturnValue:Box Structure)
   ) -> (Output_Get:Box Structure)
   n19 = Vector Length(A=@n18.BoxExtent)
   > Branch(Condition="float > float"(A=@n19, B=1))
   n20 = Get Actor Bounds(self=Get Current Nearest() -> (Current Nearest:Actor Object Reference), bOnlyCollidingComponents=false)
   > Branch(Condition="Nearly Equal (Float)"(A=@n19, B=Vector Length(A=@n20.BoxExtent), ErrorTolerance=Get Bbox Thresh() -> (Bbox Thresh:double)))
   n21 = Get Current Nearest()
   n22 = Get Current Actor()
   n23 = Get Xform Thresh() -> (Xform Thresh:double)
   > Branch(Condition="Nearly Equal (Transform)"!(
     A:"Transform (by ref)"=Get Actor Transform(self=@n22)
     B:"Transform (by ref)"=Get Actor Transform(self=@n21)
     LocationTolerance:float=100
     RotationTolerance:float=@n23
     Scale3DTolerance:float=@n23
    ) -> (ReturnValue:bool))
   n24 = Find(TargetMap=Get Overlapping Bounds Dict(), Key=Get Current Object Name() -> (Current Object Name:String))
   > n25 = Branch(Condition=@n24.ReturnValue)
   n26 = Break S Overlap Data Find Overlap!(S_OverlapData_FindOverlap:"S Overlap Data Find Overlap Structure (by ref)"=@n24) -> (
    box_14_#1:Box Structure
    labels_9_#2:Array of Strings
    descs_12_#3:Array of Actor Desc Structures
    actors_18_#4:Array of Actor Object References
    self_21_#5:Actor Object Reference
   ) {is=title K2Node_BreakStruct}
   > n27 = Set Temp Labels(Temp Labels:Array of Strings=@n26.labels_9_#2) -> (Output_Get:Array of Strings)
   > n28 = Set Temp Source Actors(Temp Source Actors:Array of Actor Object References=@n26.actors_18_#4) -> (Output_Get:Array of Actor Object References)
   > Add Unique(TargetArray:Array of Actor Object References=@n28, NewItem:"Actor Object Reference (by ref)"=Get Current Nearest())
   > Add(TargetArray=@n27, NewItem=GetActorLabel(self=Get Current Nearest(), bCreateIfNone=false)) {is=title K2Node_CallArrayFunction}
   n29 = Get Overlapping Bounds Dict()
   > Remove!(execute, TargetMap:Map of Strings to S Overlap Data Find Overlap Structures=@n29, Key:"String (by ref)"=Get Current Object Name()) -> (then, ReturnValue:bool)
   > Add!(
    execute
    TargetMap:Map of Strings to S Overlap Data Find Overlap Structures=@n29
    Key:"String (by ref)"=Get Current Object Name()
    Value:"S Overlap Data Find Overlap Structure (by ref)"=Make S Overlap Data Find Overlap!(
     box_14_#1:Box Structure=@n26
     labels_9_#2:Array of Strings=Get Temp Labels() -> (Temp Labels:Array of Strings)
     descs_12_#3:Array of Actor Desc Structures=@n26.descs_12_#3
     actors_18_#4:Array of Actor Object References=Get Temp Source Actors() -> (Temp Source Actors:Array of Actor Object References)
     self_21_#5:Actor Object Reference
    ) -> (S_OverlapData_FindOverlap:S Overlap Data Find Overlap Structure) {is=title K2Node_MakeStruct}
   ) -> (then) {is=title K2Node_CallFunction}
   n30 = Get Current Actor()
   n31 = Get Current Nearest()
   n32 = GetActorLabel(self=Get Current Actor(), bCreateIfNone=false)
   n33 = GetActorLabel(self=Get Current Nearest(), bCreateIfNone=false)
   Add(
    execute=@n25.else
    TargetMap=Get Overlapping Bounds Dict()
    Key=Get Current Object Name()
    Value=Make S Overlap Data Find Overlap(
     box_14_#1=Get Cur Actor Box() -> (Cur Actor Box:Box Structure)
     labels_9_#2=Make Array!("[0]":String=@n32, "[1]":String=@n33) -> (Array:Array of Strings)
     actors_18_#4=Make Array("[0]":Actor Object Reference=@n30, "[1]":Actor Object Reference=@n31) -> (Array:Array of Actor Object References)
    )
   )
   Set Dist Thresh(execute=@n4.NewEnumerator1, Dist Thresh=Get Rough Dist Thresh() -> (Rough Dist Thresh:double))
   > Set Bbox Thresh(Bbox Thresh=Get Rough Bbox Thresh() -> (Rough Bbox Thresh:double))
   > n34 = Set Xform Thresh(Xform Thresh=Get Rough Xform Thresh() -> (Rough Xform Thresh:double))
   Set Num With Mult Actors(execute=@n3.then_3, Num With Mult Actors:int=0) -> (Output_Get:int=0)
   > n35 = "For Each Loop (Map)"!(execute:Exec, MapPin:Map of Strings to S Overlap Data Find Overlap Structures=@n1, BreakPin:Exec) -> (then:Exec, KeyPin:String, ValuePin:S Overlap Data Find Overlap Structure, CompletedPin:Exec)
   n36 = Break S Overlap Data Find Overlap(S_OverlapData_FindOverlap=@n35.ValuePin)
   > Append Array!(
    execute
    TargetArray:Array of Actor Object References=Get AllOverlappingActors()
    SourceArray:Array of Actor Object References=@n36.actors_18_#4
   ) -> (then)
   > Branch(Condition="integer > integer"(A=Length(TargetArray=@n36.actors_18_#4) {is=title K2Node_CallArrayFunction}, B=1))
   > Set Num With Mult Actors(Num With Mult Actors=int + int(A=Get Num With Mult Actors() -> (Num With Mult Actors=0))) -> (Output_Get=0)
   n37 = "For Each Loop (Map)"(execute=@n3.then_4, MapPin=Get All Bounds Dict())
   > Clear(TargetArray:Array of Transforms=Get This Cluster Transforms() -> (This Cluster Transforms:Array of Transforms))
   n38 = Break S Overlap Data Find Overlap(S_OverlapData_FindOverlap=@n37.ValuePin)
   > Branch(Condition="integer > integer"(A=Length(TargetArray=@n38.actors_18_#4), B=1))
   > Set This Cluster Has Diff Transforms(This Cluster Has Diff Transforms:bool=false) -> (Output_Get:bool)
   > Sequence!(execute) -> (then_0, then_1)
   > n39 = Switch on E_Precision_FindOverlap(Selection=Get Overlap Threshold(self=Get Settings()) -> (Overlap Threshold=NewEnumerator0))
   > n40 = For Each Loop(Array=Break S Overlap Data Find Overlap(S_OverlapData_FindOverlap=@n37.ValuePin).actors_18_#4)
   n41 = Get This Cluster Transforms()
   n42 = For Each Loop with Break!(Exec:Exec=@n40.Completed, Array:Array of Transforms=@n41, Break:Exec=@n45.then) -> (LoopBody:Exec, Array Element:Transform, Array Index:int, Completed:Exec)
   n43 = Branch(execute=[@n42.Completed, @n39.NewEnumerator1], Condition=Get This Cluster Has Diff Transforms() -> (This Cluster Has Diff Transforms:bool))
   Add(execute=@n43.else, TargetMap=Get Overlapping Bounds Dict(), Key=@n37, Value=@n37.ValuePin)
   > Append Array(
    TargetArray=Get AllOverlappingActors()
    SourceArray=Break S Overlap Data Find Overlap(S_OverlapData_FindOverlap=@n37.ValuePin).actors_18_#4
   )
   n44 = Branch(
    execute=@n42.LoopBody
    Condition="Nearly Equal (Transform)"(
     A="Get (a ref)"(Array:Array of Transforms=@n41, Dimension 1=0) -> (Output:"Transform (by ref)")
     B=@n42
     LocationTolerance=0.0001
     RotationTolerance=0.0001
     Scale3DTolerance=0.0001
    )
   )
   n45 = Set This Cluster Has Diff Transforms(execute=@n44.else, This Cluster Has Diff Transforms=true)
   Add(
    execute=@n40.LoopBody
    TargetArray:Array of Transforms=Get This Cluster Transforms()
    NewItem:"Transform (by ref)"=Get Actor Transform(self=@n40)
   ) {is=title K2Node_CallArrayFunction}
   n46 = Switch on E_Precision_FindOverlap(execute=@n3.then_6, Selection=Get Overlap Threshold(self=Get Settings()) -> (Overlap Threshold=NewEnumerator0))
   Clear(execute=@n46.NewEnumerator1, TargetMap=Get Overlapping Bounds Dict()) {is=title K2Node_CallFunction}
   > Clear(TargetArray:Array of Actor Object References=Get AllOverlappingActors()) {is=title K2Node_CallArrayFunction}
   > n47 = "For Each Loop (Map)"(MapPin=Get All Bounds Dict())
   n48 = Break S Overlap Data Find Overlap(S_OverlapData_FindOverlap=@n47.ValuePin)
   > PrintText(InText=Format Text!(
     Format="LOCTEXT(\":\\r\\nkey: {k}\\r\\nnum actors: {n}\")"
     k:String=@n47
     n:int=Length(TargetArray=@n48.actors_18_#4)
    ) -> (Result))
   > n49 = For Each Loop(Array=@n48.actors_18_#4)
   > n50 = Branch(Condition="Equal (Integer)"(A=@n49.Array Index))
   n51 = Get Actor Location(self=@n49)
   n52 = Get Actor Location(self="Get (a copy)"(Array:Array of Actor Object References=@n48.actors_18_#4) -> (Output:Actor Object Reference))
   Branch(
    execute=@n50.else
    Condition="float <= float"(A="Distance (Vector)"!(V1:Vector=@n52, V2:Vector=@n51) -> (ReturnValue:double)) -> (ReturnValue=~)
   )
   > Add(TargetMap=Get Overlapping Bounds Dict(), Key=@n47, Value=@n47.ValuePin) {is=title K2Node_CallFunction}
   > Append Array(
    TargetArray=Get AllOverlappingActors()
    SourceArray=Break S Overlap Data Find Overlap(S_OverlapData_FindOverlap=@n47.ValuePin).actors_18_#4
   )
   Clear(execute=@n3.then_1, TargetMap=Get All Bounds Dict()) {is=title K2Node_CallFunction}
   > Clear(TargetArray=Get AllOverlappingActors()) {is=title K2Node_CallArrayFunction}
   > Clear(TargetMap=Get Overlapping Bounds Dict()) {is=title K2Node_CallFunction}
   > Clear(TargetArray:Array of Actor Desc Structures=Get FilteredDescriptions() -> (FilteredDescriptions:Array of Actor Desc Structures)) {is=title K2Node_CallArrayFunction}
   > n53 = GetAllLevelActors(self=Get EditorActorSubsystem())
   > n54 = Set AllLevelActors(AllLevelActors:Array of Actor Object References=@n53) -> (Output_Get:Array of Actor Object References)
   > n55 = For Each Loop(Array=@n54)
   > n56 = Set Current Actor(Current Actor=@n55)
   > Set FoundIgnoreString(FoundIgnoreString:bool=false) -> (Output_Get:bool)
   > n57 = For Each Loop with Break(Array:Array of Strings=Get DescIgnoreStrings() -> (DescIgnoreStrings:Array of Strings), Break=@n67.then) -> (Array Element:String)
   n58 = Branch(execute=@n57.Completed, Condition=Get FoundIgnoreString() -> (FoundIgnoreString:bool))
   > Add Unique(TargetArray=Get IgnoreTheseActors(), NewItem=Get Current Actor())
   Branch(execute=@n58.else, Condition=Get Ignore Devices!(self:BP STPS Find Overlap Object Reference=Get Settings()) -> (Ignore Devices:bool))
   > n59 = For Each Loop with Break(Array=Get ClassIgnoreStrings() -> (ClassIgnoreStrings:Array of Strings), Break=@n62.then)
   Branch(execute=@n59.Completed, Condition=Get FoundIgnoreString())
   > Add Unique(TargetArray=Get IgnoreTheseActors(), NewItem=Get Current Actor())
   n60 = "Get Class Path Name (Top Level Asset Path)"!(Class:Object Class Reference=Get Class(Object=Get Current Actor())) -> (ReturnValue_PathString:String) {is=title K2Node_CallFunction}
   n61 = Branch(execute=@n59.LoopBody, Condition="integer >= integer"(A=FindSubstring(SearchIn=@n60, Substring=@n59)))
   n62 = Set FoundIgnoreString(execute=[@n61.then, @n65.then], FoundIgnoreString=true)
   n63 = Cast To BuildingActor!(execute=@n61.else, Object=Get Current Actor()) -> (then, CastFailed, AsBuilding Actor:Building Actor Object Reference) {is=title K2Node_DynamicCast}
   > n64 = IsCreativeDevice!(execute, BuildingActor:Building Actor Object Reference=@n63) -> (then, ReturnValue:bool)
   > n65 = Branch(Condition=@n64)
   n66 = GetObjectName(Object=@n56)
   Branch(execute=@n57.LoopBody, Condition="integer >= integer"(A=FindSubstring(SearchIn=@n66, Substring=@n57)))
   > n67 = Set FoundIgnoreString(FoundIgnoreString=true)
   n68 = GetActorLabel(self=@n12.ReturnValue, bCreateIfNone=false)
   n69 = Branch(Condition="integer >= integer"(A=FindSubstring(SearchIn=@n68, Substring=windows23)))
   > n70 = PrintString!(execute, InString=@n11, bPrintToScreen, bPrintToLog, TextColor, Duration, Key) -> (then)
   For Loop(LastIndex=int + int(A=Length(TargetArray:Array of Wildcards)))
  ]}
 NewMacro {nodes [
   n1 = Inputs!() -> (Index:int, A:double=~)
   n2 = float * float(A=@n1.A, B=Get Current Precision() -> (Current Precision:double)) {is=title K2Node_PromotableOperator}
   n3 = "Sign (Float)"!(A:double=@n1.A) -> (ReturnValue:double)
   n4 = "Absolute (Float)"(A=@n2)
   n5 = Select!(
    Option 0:int=Floor(A=@n2) {is=title K2Node_CallFunction}
    Option 1:int=Round(A=@n2) {is=title K2Node_CallFunction}
    Option 2:int=Truncate(A=@n2)
    Option 3:int=Ceil(A=@n2) {is=title K2Node_CallFunction}
    Option 4:int=int * int(A=Round(A=@n4), B=Truncate(A=@n3))
    Index:int=@n1
   ) -> (ReturnValue:int)
   Outputs!(ReturnValue:double=float / float(A="To Float (Integer)"(InInt=@n5), B=Get Current Precision()))
  ]}
 On Begin Hover Sequence Hit Test {nodes [
   On Begin Hover Sequence Hit Test!() -> (then, CurrentPos:Input Device Ray Structure, Modifiers:Scriptable Tool Modifier States Structure) {is=title K2Node_FunctionEntry}
   > SetFocusInViewport(self:Scriptable Interactive Tool Object Reference)
   > Return Node!(
    execute
    ReturnValue_bHit:bool=False
    ReturnValue_HitDepth:double=340282346638528859811704183484516925440
    ReturnValue_HitNormal:Vector="0,0,1"
    ReturnValue_bHasHitNormal:bool=False
    ReturnValue_HitIdentifier:int=0
    ReturnValue_HitObject:Object Reference
   )
  ]}
 On Cycle Overlapping Viz Pressed {nodes [
   On Cycle Overlapping Viz Pressed!() -> (then, Key:Key Structure=None, Modifiers:Scriptable Tool Modifier States Structure) {is=title K2Node_FunctionEntry}
   n1 = Get Settings()
   > Set Show Only Overlapping(
    Show Only Overlapping=NOT Boolean(A=Get Show Only Overlapping!(self:BP STPS Find Overlap Object Reference=@n1) -> (Show Only Overlapping:bool))
    self=@n1
   )
  ]}
 On Focus Next Pressed {nodes [
   On Focus Next Pressed!() -> (then, Key:Key Structure=None, Modifiers:Scriptable Tool Modifier States Structure) {is=title K2Node_FunctionEntry}
   > OnFocusNextClicked()
  ]}
 On Reset View Pressed {nodes [
   On Reset View Pressed!() -> (then, Key:Key Structure=None, Modifiers:Scriptable Tool Modifier States Structure) {is=title K2Node_FunctionEntry}
   > OnRestoreVisibilityClicked()
  ]}
 On Select All Overlapping Clicked {nodes [
   On Select All Overlapping Clicked() {is=title K2Node_FunctionEntry}
   > Set Selected Level Actors Ignore Eyeball!(execute, Actors To Select:Array of Actor Object References=Get AllOverlappingActors()) -> (then) {is=title K2Node_CallFunction}
  ]}
 On Show Only Overlapping Changed {nodes [
   On Show Only Overlapping Changed!() -> (then, PropertySet:Scriptable Interactive Tool Property Set Object Reference, PropertyName:String) {is=title K2Node_FunctionEntry}
   n1 = Get Settings()
   > Set Show Only Overlapping State(Show Only Overlapping State:bool=NOT Boolean(A=Get Show Only Overlapping(self=@n1))) -> (Output_Get:bool)
   > OnShowOnlyOverlappingCycle()
  ]}
 OnFindOverlappingObjectsClicked {nodes [
   OnFindOverlappingObjectsClicked() {is=title K2Node_FunctionEntry}
   > OnRestoreVisibilityClicked()
   > n1 = GetOverlappingActors!(execute, self) -> (
    then
    All Overlapping Actors:Array of Actor Object References
    Overlapping Bounds Dict:Map of Strings to S Overlap Data Find Overlap Structures
   ) {is=title K2Node_CallFunction}
   > Set AllOverlappingActors(AllOverlappingActors:Array of Actor Object References=@n1) -> (Output_Get:Array of Actor Object References)
   > Set Total Clusters Overlapping(Total Clusters Overlapping=Length(TargetMap=@n1.Overlapping Bounds Dict) -> (ReturnValue=0) {is=title K2Node_CallFunction}) -> (Output_Get=0)
   > RemoveAllLines(self=Get Line Set For Indicators())
   > RemoveAllLines(self=Get Line Set For Bbox())
   > Clear(TargetArray:Array of Scriptable Tool Line Object References=Get Box Lines() -> (Box Lines:Array of Scriptable Tool Line Object References))
   > Set Cur Cluster(Cur Cluster:int=0) -> (Output_Get:int=0)
   > Set Total Actors Overlapping(Total Actors Overlapping=0) -> (Output_Get=0)
   > n2 = "For Each Loop (Map)"(MapPin=@n1.Overlapping Bounds Dict)
   n3 = Break S Overlap Data Find Overlap(S_OverlapData_FindOverlap=@n2.ValuePin)
   > Set Total Actors Overlapping(Total Actors Overlapping=int + int(
     A=Length(TargetArray:Array of Actor Object References=@n3.actors_18_#4) {is=title K2Node_CallArrayFunction}
     B=Get Total Actors Overlapping() -> (Total Actors Overlapping:int=0)
    )) -> (Output_Get=0)
   n4 = GetBoxCenterSize!(Box:Box Structure=@n3) -> (Center:Vector, Dimensions:Vector)
   > Draw Vertical Indicators!(
    execute
    Line Set:Scriptable Tool Line Set Object Reference=Get Line Set For Indicators()
    Start:Vector=@n4
    Height:double=10000
    Color:Color Structure=Make Color!(B:Byte=0, G:Byte=255, R:Byte=0, A:Byte=255) -> (Color:Color Structure) {is=title K2Node_MakeStruct}
    Thickness:float=10
    DepthBias:float=0
   ) -> (then)
   > n5 = Make BBox Lines in Lineset!(execute, Line Set For BBox:Scriptable Tool Line Set Object Reference=Get Line Set For Bbox()) -> (then, New Box Lines:Array of Scriptable Tool Line Object References)
   > Append Array(TargetArray:Array of Scriptable Tool Line Object References=Get Box Lines(), SourceArray:Array of Scriptable Tool Line Object References=@n5)
   > DrawBBox!(
    execute
    BoundOrigin:Vector=@n4
    BoundExtent:Vector=vector / vector(A:Vector=@n4.Dimensions, B:Vector="2,2,2") -> (ReturnValue:Vector)
    Box Lines:Array of Scriptable Tool Line Object References=@n5
    Color:Color Structure=Make Color(B=255, G=165, R=255, A=255)
    Thickness:float=4
    DepthBias:float=0
   ) -> (then)
   > Increment Int(Value=Get Cur Cluster() -> (Cur Cluster:int=0))
   Set Text(
    execute=@n2.CompletedPin
    Text="To Text (Integer)"(Value=Get Total Clusters Overlapping() -> (Total Clusters Overlapping:int=0))
    self=Get Value!(self:BP UW Readout Line Object Reference=Get Overlap Groups Line() -> (Overlap Groups Line:BP UW Readout Line Object Reference)) -> (Value:Common Text Object Reference)
   )
   > Set Text(
    Text="To Text (Integer)"(Value=Get Total Actors Overlapping() -> (Total Actors Overlapping=0))
    self=Get Value(self=Get Overlap Total Line() -> (Overlap Total Line:BP UW Readout Line Object Reference))
   )
   n6 = For Each Loop(Array:Array of Scriptable Tool Line Set Object References=Get All BBox Line Sets() -> (All BBox Line Sets:Array of Scriptable Tool Line Set Object References)) -> (Array Element:Scriptable Tool Line Set Object Reference)
   > RemoveAllLines(self=@n6)
   Length(TargetArray:Array of Actor Desc Structures=@n3.descs_12_#3)
  ]}
 OnFocusNextClicked {nodes [
   OnFocusNextClicked() {is=title K2Node_FunctionEntry}
   n1 = Get Overlapping Bounds Dict()
   n2 = Length(TargetMap=@n1) -> (ReturnValue=0) {is=title K2Node_CallFunction}
   > Branch(Condition="integer > integer"(A=@n2))
   n3 = "% (Integer)"(A=int + int(A=Get CurClusterFocused() -> (CurClusterFocused:int=0)), B=@n2) -> (ReturnValue=0)
   > Set CurClusterFocused(CurClusterFocused=@n3) -> (Output_Get=0)
   n4 = Break S Overlap Data Find Overlap(S_OverlapData_FindOverlap=Get Key Value By Index!(TargetMap:Map of Strings to S Overlap Data Find Overlap Structures=@n1, Index:int=@n3) -> (Key:String, Value:S Overlap Data Find Overlap Structure) {is=title K2Node_CallFunction}.Value)
   > SetSelectedLevelActors(self=Get EditorActorSubsystem(), ActorsToSelect=@n4.actors_18_#4)
   n5 = GetBoxCenterSize(Box=@n4)
   n6 = vector + vector(A=@n5, B=vector * vector(A:double=GetMaxElement!(A:Vector=@n5.Dimensions) -> (ReturnValue:double), B:Vector="1,1,1") -> (ReturnValue=~))
   > SetLevelViewportCameraInfo(
    self=Get UnrealEditorSubsystem()
    CameraLocation=@n6
    CameraRotation=FindLookAtRotation!(Start:"Vector (by ref)"=@n6, Target:"Vector (by ref)"=@n5) -> (ReturnValue:Rotator)
   )
  ]}
 OnRestoreVisibilityClicked {nodes [
   OnRestoreVisibilityClicked() {is=title K2Node_FunctionEntry}
   > Set Show Only Overlapping(Show Only Overlapping=false, self=Get Settings())
   > n1 = "For Each Loop (Map)"(MapPin:Map of Actor Object References to Booleans=Get Actor Temp Hidden At Start() -> (Actor Temp Hidden At Start:Map of Actor Object References to Booleans)) -> (KeyPin:Actor Object Reference, ValuePin:bool)
   > SetIsTemporarilyHiddenInEditor!(execute, self:Actor Object Reference=@n1, bIsHidden:bool=@n1.ValuePin) -> (then)
   > SetLevelViewportCameraInfo(self=Get UnrealEditorSubsystem(), CameraLocation=Get Viewport Cam At Start Location(), CameraRotation=Get Viewport Cam At Start Rotation())
   > SetEditorActiveViewportViewMode!(execute, Index:EViewModeIndex Enum=Get View Mode Before OnShowOnly() -> (View Mode Before OnShowOnly:EViewModeIndex Enum=VMI_BrushWireframe)) -> (then)
  ]}
 OnShowOnlyOverlappingCycle {nodes [
   n1 = GetAllLevelActors(execute=@n9.then, self=Get EditorActorSubsystem())
   > n2 = For Each Loop(Array:Array of Actor Object References=@n1) -> (Array Element:Actor Object Reference)
   n3 = AddLogMessage!(
    execute
    self:Scriptable Interactive Tool Object Reference
    Message:Text=Format Text!(Format="LOCTEXT(\"Is Hidden  ON: {a}\")", a:Actor Object Reference=@n2) -> (Result)
    bHighlighted:bool=false
   ) -> (then)
   SetIsTemporarilyHiddenInEditor(execute=[@n3.then, @n11.else], self=@n2, bIsHidden=true)
   n4 = AddLogMessage(Message=Format Text(Format="LOCTEXT(\"Is Hidden OFF: {a}\")", a=@n2), bHighlighted=false)
   SetIsTemporarilyHiddenInEditor(execute=[@n4.then, @n11.then], self=@n2, bIsHidden=false)
   OnShowOnlyOverlappingCycle() {is=title K2Node_FunctionEntry}
   > n5 = Branch(Condition=Get Show Only Overlapping State() -> (Show Only Overlapping State:bool))
   > Set Show Only Overlapping State(Show Only Overlapping State=false)
   > n6 = "For Each Loop (Map)"(MapPin=Get Actor Temp Hidden On Show Only Overlapping() -> (Actor Temp Hidden On Show Only Overlapping:Map of Actor Object References to Booleans))
   > SetIsTemporarilyHiddenInEditor(self=@n6, bIsHidden=@n6.ValuePin)
   SetEditorActiveViewportViewMode(execute=@n6.CompletedPin, Index=Get View Mode Before OnShowOnly() -> (View Mode Before OnShowOnly=VMI_BrushWireframe))
   Set Show Only Overlapping State(execute=@n5.else, Show Only Overlapping State=true)
   > n7 = GetActorsTempHidden()
   > Set Actor Temp Hidden On Show Only Overlapping(Actor Temp Hidden On Show Only Overlapping:Map of Actor Object References to Booleans=@n7) -> (Output_Get:Map of Actor Object References to Booleans)
   > n8 = GetEditorActiveViewportViewMode() -> (ReturnValue=VMI_BrushWireframe)
   > Set View Mode Before OnShowOnly(View Mode Before OnShowOnly=@n8) -> (Output_Get=VMI_BrushWireframe)
   > n9 = SetEditorActiveViewportViewMode(Index=VMI_Unlit)
   n10 = Find Item(TargetArray=Get AllOverlappingActors(), ItemToFind=@n2)
   n11 = Branch(execute=@n2.LoopBody, Condition="integer >= integer"(A=@n10))
  ]}
 PrecisionIfier {nodes [
   n1 = PrecisionIfier!() -> (then, Raw Vector:Vector=~, Precision:E_Precision_FindOverlap Enum=NewEnumerator0, Trunc method:int) {is=title K2Node_FunctionEntry}
   > Set Current Precision(Current Precision:double=Select!(
     NewEnumerator0:double=Get Precise Precision Factor() -> (Precise Precision Factor:double)
     NewEnumerator1:double=Get Rough Precision Factor() -> (Rough Precision Factor:double)
     Index:E_Precision_FindOverlap Enum=@n1.Precision
    ) -> (ReturnValue:double)) -> (Output_Get:double)
   > n2 = Branch(Condition=Get Breakpoint Global() -> (Breakpoint Global:bool))
   > n3 = PrintString(InString=Hello, bPrintToScreen=false, bPrintToLog=false)
   n4 = BreakVector(InVec=@n1)
   Return Node!(
    execute=[@n3.then, @n2.else]
    Revised Vector:Vector=MakeVector(
     X=New Macro!(Index:int=@n1.Trunc method, A:double=@n4) -> (ReturnValue:double=~) {is=title K2Node_MacroInstance}
     Y=New Macro(Index=@n1.Trunc method, A=@n4.Y) -> (ReturnValue=~)
     Z=New Macro(Index=@n1.Trunc method, A=@n4.Z) -> (ReturnValue=~)
    )
   )
  ]}
]
variables [
 ActionButtons {type=BP_UW_ActionsButton_FindOverlap_C, default=null}
 Actions Box {type=BP_UW_ExpandableBox_C}
 Actor Temp Hidden At Start {}
 Actor Temp Hidden On Show Only Overlapping {}
 All Actors Filtered Descriptions {}
 All BBox Line Sets {type="array<ScriptableToolLineSet>", default []}
 All Bounds Dict {type="map<S_OverlapData_FindOverlap>", default {}}
 AllBoundsDict {type="map<S_OverlapData_FindOverlap>", default {}}
 AllOverlappingActors {type="array<Actor>"}
 Box Lines {type="array<ScriptableToolLine>", default []}
 Breakpoint Global {type=boolean}
 ClassIgnoreStrings {type="array<string>", default [Device_, Fortnite/Devices, CreativeCoreDevices]}
 CurClusterFocused {type=integer}
 CurTime {type=number}
 Cycle Overlapping Viz Key {type=Key, default=V}
 DescIgnoreStrings {
  type="array<string>"
  default [_HLOD, GroupActor, MiniMap, Foliage, DataLayers, LevelBounds, IslandSettings, DirectionalLight, PostProcessVolume, SkyLight, PlaysetRoot]
 }
 FirstInitRun {type=boolean}
 Focus Next Hotkey {type=Key, default=C}
 Header Widget {type=BP_UW_SharedHeaderWidget_C}
 Line Set For Bbox {type=ScriptableToolLineSet, default=null}
 Line Set For Indicators {type=ScriptableToolLineSet, default=null}
 Overlap Groups Line {type=BP_UW_ReadoutLine_C, default=null}
 Overlap Total Line {type=BP_UW_ReadoutLine_C, default=null}
 Overlapping Bounds Dict {type="map<S_OverlapData_FindOverlap>", default {}}
 Precise Bbox Thresh {type=number, default=0.01}
 Precise Dist Thresh {type=number, default=1}
 Precise Precision Factor {type=number, default=1000}
 Precise Xform Thresh {type=number, default=0.0001}
 Property Set Save Key {type=string, default=Prop Set Save Key}
 Readout Box {type=BP_UW_ExpandableBox_C}
 Reset View Hotkey {type=Key, default=B}
 Rough Bbox Thresh {type=number, default=1}
 Rough Dist Thresh {type=number, default=10}
 Rough Precision Factor {type=number, default=0.01}
 Rough Xform Thresh {type=number, default=0.01}
 Settings {type=BP_STPS_FindOverlap_C, default=null}
 Settings Identifier {type=string, default=Settings}
 SharedState Settings {type=BP_STPS_SharedWidgetState_C, default=null}
 SharedStateIdentifier {type=string, default=SaveSettings}
 Shortcut Box {type=BP_UW_ExpandableBox_C}
 Shortcut Map {type="map<string>", default {Cycle Overlapping Visibility=Press V, Focus Next Overlap=Press C, Reset View=Press B}}
 Shortcut Map Keys {type="array<string>", default [Focus Next Overlap, Cycle Overlapping Visibility, Reset View]}
 ShortcutLines {type="array<BP_UW_ReadoutLine_C>", default []}
 Show Only Overlapping State {type=boolean}
 Total Actors Overlapping {type=integer}
 Total Clusters Overlapping {type=integer}
 View Mode Before OnShowOnly {type=EViewModeIndex, default=VMI_BrushWireframe}
 Viewport Cam At Start Location {type=Vector}
 Viewport Cam At Start Rotation {type=Rotator}
 "Widget Num Overlap Groups  Label" {type=string, default=Number Overlapping Clusters}
 "Widget Num Overlap Groups  Value" {type=string, default=0}
 "Widget Num Overlap Total  Label" {type=string, default=Total Overlapping Objects}
 "Widget Num Overlap Total  Value" {type=string, default=0}
]
